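# Sound cue for third-person crouch-sprint footsteps on dirt in Athena, routed to the enemy foley sound class.
# A quality-level node picks a random pool of Commando step waves: six for the first two levels, four for the last two.
# A root modulator raises pitch and varies volume per play.
# A crouch-sprint attenuation asset shapes falloff, and local and global footstep concurrency sets limit overlapping steps.
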SoundCue /Game/Athena/Sounds/Character/Footsteps/3P/Dirt/Athena_Character_Crouch_Sprint_Footstep_Dirt_3P_Cue
root=Modulator_0
nodes [
 {id=Modulator_0, children [QualityLevel_0], properties {pitchMax=1.5, pitchMin=1.3, volumeMax=1, volumeMin=0.7}}
 {id=QualityLevel_0, children [Random_1,Random_1,Random_0,Random_0]}
 {
  id=Random_0
  children [WavePlayer_6, WavePlayer_7, WavePlayer_8, WavePlayer_12]
  properties {bShouldExcludeFromBranchCulling=true, weights [1,1,1,1]}
 }
 {
  id=Random_1
  children [WavePlayer_0,WavePlayer_1,WavePlayer_2,WavePlayer_3,WavePlayer_4,WavePlayer_5]
  properties {bShouldExcludeFromBranchCulling=true, weights [1,1,1,1,1,1]}
 }
 {id=WavePlayer_0, properties {soundWaveAssetPtr=/Game/Sounds/Fortress_Foley/Commando/Commando_Step_001}}
 {id=WavePlayer_1, properties {soundWaveAssetPtr=/Game/Sounds/Fortress_Foley/Commando/Commando_Step_002}}
 {id=WavePlayer_12, properties {soundWaveAssetPtr=/Game/Sounds/Fortress_Foley/Commando/Commando_Step_004}}
 {id=WavePlayer_2, properties {soundWaveAssetPtr=/Game/Sounds/Fortress_Foley/Commando/Commando_Step_003}}
 {id=WavePlayer_3, properties {soundWaveAssetPtr=&Commando_Step_004}}
 {id=WavePlayer_4, properties {soundWaveAssetPtr=/Game/Sounds/Fortress_Foley/Commando/Commando_Step_006}}
 {id=WavePlayer_5, properties {soundWaveAssetPtr=/Game/Sounds/Fortress_Foley/Commando/Commando_Step_010}}
 {id=WavePlayer_6, properties {soundWaveAssetPtr=&Commando_Step_001}}
 {id=WavePlayer_7, properties {soundWaveAssetPtr=&Commando_Step_002}}
 {id=WavePlayer_8, properties {soundWaveAssetPtr=&Commando_Step_003}}
]
properties {
 attenuationSettings=/Game/Sounds/Attenuation/Fort_Player_Footsteps_3P_CrouchSprint_Athena_Att
 maxDistance=2560
 soundClassObject=&Fort_Player_Foley_3P_Enemy_Athena
 volumeMultiplier=0.4
 concurrencyOverrides {bLimitToOwner=true, maxCount=2, resolutionRule=StopOldest}
 concurrencySet [&Athena_Footstep_Local_Conc, &Athena_Footstep_Global_Conc]
}
