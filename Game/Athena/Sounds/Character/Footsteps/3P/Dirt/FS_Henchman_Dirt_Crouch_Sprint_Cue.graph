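# Third-person henchman footstep cue for crouch-sprinting on dirt, built from Henchman dirt run waves and routed to the enemy foley sound class.
# A Modulator feeds a quality-level switch: the first level picks randomly among eight waves, while the other three share a four-wave random subset.
# Crouch-sprint footstep attenuation, a reduced volume and local plus global footstep concurrency sets shape playback, and the cue primes on load.
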SoundCue /Game/Athena/Sounds/Character/Footsteps/3P/Dirt/FS_Henchman_Dirt_Crouch_Sprint_Cue
root=Modulator_1
nodes [
 {id=Modulator_1, children [QualityLevel_0]}
 {id=QualityLevel_0, children [Random_0,Random_1,Random_1,Random_1]}
 {
  id=Random_0
  children [WavePlayer_6, WavePlayer_7, WavePlayer_8, WavePlayer_9, WavePlayer_10, WavePlayer_11, WavePlayer_12, WavePlayer_13]
  properties {weights [1,1,1,1,1,1,1,1]}
 }
 {
  id=Random_1
  children [WavePlayer_1,WavePlayer_2,WavePlayer_3,WavePlayer_4]
  properties {bShouldExcludeFromBranchCulling=true, weights [1,1,1,1]}
 }
 {id=WavePlayer_1, properties {soundWaveAssetPtr=&/FS_Henchman_Dirt_Run_01}}
 {id=WavePlayer_10, properties {soundWaveAssetPtr=&/FS_Henchman_Dirt_Run_04}}
 {id=WavePlayer_11, properties {soundWaveAssetPtr=&/FS_Henchman_Dirt_Run_05}}
 {id=WavePlayer_12, properties {soundWaveAssetPtr=&/FS_Henchman_Dirt_Run_06}}
 {id=WavePlayer_13, properties {soundWaveAssetPtr=&/FS_Henchman_Dirt_Run_07}}
 {id=WavePlayer_2, properties {soundWaveAssetPtr=&/FS_Henchman_Dirt_Run_02}}
 {id=WavePlayer_3, properties {soundWaveAssetPtr=&/FS_Henchman_Dirt_Run_03}}
 {id=WavePlayer_4, properties {soundWaveAssetPtr=&FS_Henchman_Dirt_Run_04}}
 {id=WavePlayer_6, properties {soundWaveAssetPtr=&/FS_Henchman_Dirt_Run_08}}
 {id=WavePlayer_7, properties {soundWaveAssetPtr=&FS_Henchman_Dirt_Run_01}}
 {id=WavePlayer_8, properties {soundWaveAssetPtr=&FS_Henchman_Dirt_Run_02}}
 {id=WavePlayer_9, properties {soundWaveAssetPtr=&FS_Henchman_Dirt_Run_03}}
]
properties {
 attenuationSettings=/Game/Sounds/Attenuation/Fort_Player_Footsteps_3P_CrouchSprint_Athena_Att
 bPrimeOnLoad=true
 maxDistance=2560
 soundClassObject=&Fort_Player_Foley_3P_Enemy_Athena
 volumeMultiplier=0.4
 concurrencyOverrides {bLimitToOwner=true, maxCount=2, resolutionRule=StopOldest}
 concurrencySet [&Athena_Footstep_Local_Conc, &Athena_Footstep_Global_Conc]
}
