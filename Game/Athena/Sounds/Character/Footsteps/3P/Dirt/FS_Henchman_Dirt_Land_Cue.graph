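# Sound cue for a Henchman's third-person landing on dirt, built from walk and run footstep waves.
# A Mixer layers a random walk sample with a random run sample delayed by a short variable interval, simulating a two-part impact.
# The mix feeds a Modulator, with player-landing attenuation, the enemy foley sound class and the global footstep concurrency set.
# The cue primes on load, and the run-sample Random node is excluded from branch culling.
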SoundCue /Game/Athena/Sounds/Character/Footsteps/3P/Dirt/FS_Henchman_Dirt_Land_Cue
root=Modulator_0
nodes [
 {id=Delay_0, children [Random_2], properties {delayMax=0.11, delayMin=0.08}}
 {id=Mixer_1, children [Random_1,Delay_0], properties {inputVolume [1,1]}}
 {id=Modulator_0, children [Mixer_1]}
 {id=Random_1, children [WavePlayer_12, WavePlayer_13, WavePlayer_14, WavePlayer_15], properties {weights [1,1,1,1]}}
 {id=Random_2, children [WavePlayer_16, WavePlayer_17, WavePlayer_18], properties {bShouldExcludeFromBranchCulling=true, weights [1,1,1]}}
 {id=WavePlayer_12, properties {soundWaveAssetPtr=&/FS_Henchman_Dirt_Walk_01}}
 {id=WavePlayer_13, properties {soundWaveAssetPtr=&/FS_Henchman_Dirt_Walk_02}}
 {id=WavePlayer_14, properties {soundWaveAssetPtr=&/FS_Henchman_Dirt_Walk_03}}
 {id=WavePlayer_15, properties {soundWaveAssetPtr=&/FS_Henchman_Dirt_Walk_04}}
 {id=WavePlayer_16, properties {soundWaveAssetPtr=&/FS_Henchman_Dirt_Run_01}}
 {id=WavePlayer_17, properties {soundWaveAssetPtr=&/FS_Henchman_Dirt_Run_02}}
 {id=WavePlayer_18, properties {soundWaveAssetPtr=&/FS_Henchman_Dirt_Run_03}}
]
properties {
 attenuationSettings=/Game/Sounds/Attenuation/Fort_Player_Land_3P_Athena_Att
 bPrimeOnLoad=true
 maxDistance=3584
 soundClassObject=&Fort_Player_Foley_3P_Enemy_Athena
 volumeMultiplier=1
 concurrencyOverrides {bLimitToOwner=true, maxCount=4, resolutionRule=StopOldest}
 concurrencySet [&Athena_Footstep_Global_Conc]
}
