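# Sound cue for a Henchman's hard landing on dirt, heard in third person through the enemy foley sound class and landing attenuation.
# Each mix layers a quieter Commando landing sound with a louder, briefly delayed random dirt run footstep, all under a root modulator.
# A quality switch gives the first two levels four footstep variants and the last two only three, the latter random node excluded from branch culling.
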SoundCue /Game/Athena/Sounds/Character/Footsteps/3P/Dirt/FS_Henchman_Dirt_Land_Hard_Cue
root=Modulator_0
nodes [
 {id=Delay_0, children [Random_2], properties {delayMax=0.11, delayMin=0.08}}
 {id=Delay_1, children [Random_0], properties {delayMax=0.11, delayMin=0.08}}
 {id=Mixer_0, children [WavePlayer_1,Delay_1], properties {inputVolume [0.5,1.2]}}
 {id=Mixer_1, children [WavePlayer_5,Delay_0], properties {inputVolume [0.5,1.2]}}
 {id=Modulator_0, children [QualityLevel_0]}
 {id=QualityLevel_0, children [Mixer_1,Mixer_1,Mixer_0,Mixer_0]}
 {id=Random_0, children [WavePlayer_2,WavePlayer_3,WavePlayer_4], properties {bShouldExcludeFromBranchCulling=true, weights [1,1,1]}}
 {id=Random_2, children [WavePlayer_16, WavePlayer_17, WavePlayer_18, WavePlayer_0], properties {weights [1,1,1,1]}}
 {id=WavePlayer_0, properties {soundWaveAssetPtr=&/FS_Henchman_Dirt_Run_04}}
 {id=WavePlayer_1, properties {soundWaveAssetPtr=/Game/Sounds/Fortress_Foley/Commando/Commando_Land_005}}
 {id=WavePlayer_16, properties {soundWaveAssetPtr=&/FS_Henchman_Dirt_Run_01}}
 {id=WavePlayer_17, properties {soundWaveAssetPtr=&/FS_Henchman_Dirt_Run_02}}
 {id=WavePlayer_18, properties {soundWaveAssetPtr=&/FS_Henchman_Dirt_Run_03}}
 {id=WavePlayer_2, properties {soundWaveAssetPtr=&FS_Henchman_Dirt_Run_01}}
 {id=WavePlayer_3, properties {soundWaveAssetPtr=&FS_Henchman_Dirt_Run_02}}
 {id=WavePlayer_4, properties {soundWaveAssetPtr=&FS_Henchman_Dirt_Run_03}}
 {id=WavePlayer_5, properties {soundWaveAssetPtr=&Commando_Land_005}}
]
properties {
 attenuationSettings=/Game/Sounds/Attenuation/Fort_Player_Land_3P_Athena_Att
 bPrimeOnLoad=true
 maxDistance=3584
 soundClassObject=&Fort_Player_Foley_3P_Enemy_Athena
 volumeMultiplier=1.4
 concurrencyOverrides {bLimitToOwner=true, maxCount=4, resolutionRule=StopOldest}
 concurrencySet [&Athena_Footstep_Global_Conc]
}
